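# Master surface material for Bridge Megascans presets, a Default Lit material that samples its texture maps as virtual textures.
# MF_Tiling supplies UVs to the Albedo, Metalness, Normal and packed ARD (AO/R/D) parameters, which default to placeholder textures.
# MF_MapAdjustments receives these samples plus the tiling rotation angle for the tangent normal, in a material-attributes setup.
# Both AO and Roughness inputs read ARD's red channel, leaving green and blue unused.
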Material /Bridge/MSPresets/M_MS_Surface_Material_VT/M_MS_Surface_Material_VT
graph {
 properties {bUseMaterialAttributes=true, pixelDepthOffsetMode=PDOM_Legacy, usedShadingModels=Default Lit}
 nodes [
  n1 = MaterialFunctionCall() {materialFunction=/Bridge/MSPresets/M_MS_Default_Material/Functions/MF_Tiling}
  n2 = TextureSampleParameter2D(UVs=@n1.Result) {
   desc=AO/R/D
   group=07 - Texture Maps
   parameterName=ARD
   samplerType=SAMPLERTYPE_VirtualLinearColor
   sortPriority=2
   texture=/Bridge/MSPresets/MSVTTextures/WhitePlaceholder
  }
  n3 = TextureSampleParameter2D(UVs=@n1.Result) {
   group=07 - Texture Maps
   parameterName=Albedo
   samplerSource=SSM_Wrap_WorldGroupSettings
   samplerType=SAMPLERTYPE_VirtualColor
   sortPriority=0
   texture=/Bridge/MSPresets/MSVTTextures/DefaultDiffuse
  }
  n4 = TextureSampleParameter2D(UVs=@n1.Result) {
   group=07 - Texture Maps
   parameterName=Metalness
   samplerSource=SSM_Wrap_WorldGroupSettings
   samplerType=SAMPLERTYPE_VirtualLinearColor
   sortPriority=1
   texture=/Bridge/MSPresets/MSVTTextures/BlackPlaceholder
  }
  n5 = TextureSampleParameter2D(UVs=@n1.Result) {
   group=07 - Texture Maps
   parameterName=Normal
   samplerSource=SSM_Wrap_WorldGroupSettings
   samplerType=SAMPLERTYPE_VirtualNormal
   sortPriority=3
   texture=/Bridge/MSPresets/MSVTTextures/Placeholder_Normal
  }
  MaterialFunctionCall!(AO=@n2.R, Albedo=@n3.RGB, Metallic=@n4.RGB, Roughness=@n2.R, Tangent Normal=@n5.RGB, Tangent Normal Rotation Angle=@n1.Rotation Angle) {materialFunction=/Bridge/MSPresets/M_MS_Default_Material/Functions/MF_MapAdjustments}
 ]
}
